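# Translucent unlit Niagara sprite material for the Voltaic Huntress pickaxe, drawing glitchy glowing binary digits.
# Dynamic Param2 blends UVs toward a hologram glitch-grid function before sampling the digit texture's red channel.
# A step threshold separates solid digit cores from a Min/Max-remapped glow, tinted by Color and Glow and multiplied by particle color and alpha.
# Param1 has no visible effect because both inputs of its lerp are identical.
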Material /BRCosmetics/Weapons/FORT_Melee/Pickaxe_VoltaicHuntress/FX/Circuit/V1/M_VoltaicHuntress_BadDigits
graph {
 properties {bUsedWithNiagaraSprites=true, blendMode=BLEND_Translucent, pixelDepthOffsetMode=PDOM_Legacy, shadingModel=MSM_Unlit, usedShadingModels=Unlit}
 nodes [
  n1 = TextureSample(UVs=LinearInterpolate(
    A=TextureCoordinate()
    B=MaterialFunctionCall!(UV) {materialFunction=/BRCosmetics/Weapons/FORT_Melee/Pickaxe_VoltaicHuntress/FX/Circuit/Functions/MF_Holog_Glitch_Grid}.Result
    Alpha=DynamicParameter().Param2
   )) {texture=&/T_BinaryDigits}
  n2 = LinearInterpolate(A=@n1.R, B=@n1.R, Alpha=DynamicParameter().Param1)
  n3 = ScalarParameter() {parameterName=Min}
  n4 = ScalarParameter() {defaultValue=0.2, parameterName=Max}
  n5 = OneMinus(None=Step(Y=@n2, X=ScalarParameter() {defaultValue=0.2, parameterName=Step}) {constX=0.2})
  n6 = Subtract(A=Divide(A=Subtract(A=@n2, B=@n3), B=Subtract(A=@n4, B=@n3)), B=@n5)
  n7 = ParticleColor()
  Multiply(A=Add(A=Saturate(None=@n6), B=@n5), B=@n7.A)
  n8 = VectorParameter() {parameterName=Glow, defaultValue=<rgba 0.0445 0.5 0.2042 1>}
  Multiply(
   A=LinearInterpolate(A=Multiply(A=@n8.RGB, B=@n6), B=VectorParameter() {parameterName=Color, defaultValue=<rgba 0.0201 1 0.244 1>}.RGB, Alpha=@n5)
   B=@n7.RGB
  )
 ]
}
